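# Sound cue for the in-world XP level-up sound, routed through the Fortnite UI sound class.
# A Random node picks between two level-up wave variations with equal weight.
# It is spatialized through a borrowed forged-consumable in-world attenuation asset.
# A concurrency override caps simultaneous instances at a small count, stopping the oldest when exceeded.
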SoundCue /FNGameplayCues/Sounds/Fort_UI_Sounds/Gameplay/XP/Fort_XP_LevelUp_InWorld_Cue
root=Random_0
nodes [
 {id=Random_0, children [WavePlayer_1,WavePlayer_2], properties {weights [1,1]}}
 {id=WavePlayer_1, properties {soundWaveAssetPtr=&/S11XP_LevelUp_InWorld_01}}
 {id=WavePlayer_2, properties {soundWaveAssetPtr=&/S11XP_LevelUp_InWorld_02}}
]
properties {
 attenuationSettings=/Game/Sounds/Attenuation/Fort_ForgedConsumable_InWorld_Att
 bOverrideConcurrency=true
 maxDistance=1000
 soundClassObject=&Fort_UI
 volumeMultiplier=1.008
 concurrencyOverrides {maxCount=2, resolutionRule=StopOldest}
}
